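# Material function for the ChicleVeil backpack that layers hue-shifted base color and emissive glow over incoming material attributes.
# Mask texture channels, the TAMask2 input and pre-skinned position thresholds define tattoo, crystal, blade and weapon regions.
# Fresnel and two opposite-panning fire noise samples, combined by Max, provide glow masks alongside GlowAmount and OnKill inputs.
# A saturated sum of masks blends the modified BaseColor and EmissiveColor attributes with a second attribute set.
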MaterialFunction /BRCosmetics/Accessories/FORT_Backpacks/Backpack_F_MED_ChicleVeil/Materials/MF_Chicleveil_Backpack
graph {nodes [
  n1 = NamedRerouteUsage()
  n2 = NamedRerouteUsage()
  n3 = LinearInterpolate(A=NamedRerouteUsage(), Alpha=NamedRerouteUsage()) {constB=0}
  n4 = Add(A=Add(A=NamedRerouteUsage(), B=NamedRerouteUsage()), B=NamedRerouteUsage())
  n5 = NamedRerouteUsage()
  n6 = Multiply(A=Add(A=Add(A=@n4, B=@n5) {constB=0}, B=NamedRerouteUsage()), B=Add(A=NamedRerouteUsage()))
  n7 = Add(A=NamedRerouteUsage(), B=NamedRerouteUsage())
  FunctionOutput(None=BlendMaterialAttributes(
    A=SetMaterialAttributes!(MaterialAttributes=@n1)
    B=SetMaterialAttributes!(MaterialAttributes=@n2, BaseColor=@n3, EmissiveColor=@n6) {attributeSetTypes [#1, #2]}
    Alpha=Saturate(None=@n7)
   ))
  n8 = FunctionInput() {inputName=TAMask2, sortPriority=8}
  NamedRerouteDeclaration(None=ComponentMask(None=@n8) {g=true}) {name=CrystalMask, nodeColor=<rgba 1 0 0.8672 1>}
  n9 = TextureSample() {samplerType=SAMPLERTYPE_Masks, texture=/BRCosmetics/Accessories/FORT_Backpacks/Backpack_F_MED_ChicleVeil/Textures/T_ChicleVeil_Backpack_FX}
  NamedRerouteDeclaration(None=@n9.A) {name=AlphaFresnel, nodeColor=<rgba 0 1 0.4609 1>}
  n10 = Multiply(A=TextureCoordinate()) {constB=2}
  n11 = Time()
  n12 = Multiply(A=Constant2Vector() {g=1}) {constB=0.1}
  n13 = Add(A=@n10, B=Multiply(A=@n11, B=@n12))
  n14 = Multiply(A=@n12) {constB=-1}
  n15 = Multiply(A=Add(A=@n10, B=Multiply(A=@n11, B=@n14))) {constB=1.1}
  NamedRerouteDeclaration(None=Max(
    A=TextureSample(UVs=@n13) {samplerSource=SSM_Wrap_WorldGroupSettings, texture=/Game/Global/Textures/Default/Noise/T_FireNoise_01}.R
    B=TextureSample(UVs=@n15) {samplerSource=SSM_Wrap_WorldGroupSettings, texture=&T_FireNoise_01}.R
   )) {name=FXFresnel, nodeColor=<rgba 0.8984 1 0 1>}
  n16 = GetMaterialAttributes(None=FunctionInput() {inputType=FunctionInput_MaterialAttributes}) {attributeGetTypes [#1]}
  NamedRerouteDeclaration(None=@n16.BaseColor) {name=InBaseColor, nodeColor=<rgba 1 0.3516 0 1>}
  NamedRerouteDeclaration(None=@n9.G) {name=TAMaskR, nodeColor=<rgba 0 1 0.3672 1>}
  n17 = NamedRerouteUsage()
  n18 = NamedRerouteUsage()
  n19 = NamedRerouteDeclaration(None=Step(X=Multiply(A=@n17, B=@n18)) {constY=0.2}) {name=BladesNoCrystalAreaMask, nodeColor=<rgba 0.0781 0 1 1>}
  n20 = NamedRerouteUsage()
  NamedRerouteDeclaration(None=Saturate(None=Add(A=@n19, B=@n20))) {name=BladesMask, nodeColor=<rgba 0 1 0.3672 1>}
  n21 = NamedRerouteUsage()
  n22 = OneMinus(None=DotProduct(A=VertexNormalWS(), B=CameraVectorWS()))
  n23 = NamedRerouteUsage()
  n24 = Multiply(A=Subtract(A=Power(Base=@n22) {constExponent=4}, B=Multiply(A=@n23))) {constB=5}
  n25 = NamedRerouteDeclaration(None=@n24) {name=Fresnel, nodeColor=<rgba 0.4531 1 0 1>}
  NamedRerouteDeclaration(None=Saturate(None=Multiply(A=@n21, B=@n25))) {name=FresnelMask, nodeColor=<rgba 0.9688 1 0 1>}
  n26 = NamedRerouteUsage()
  n27 = Constant() {r=0.4}
  NamedRerouteDeclaration(None=LinearInterpolate(
    A=@n26
    B=MaterialFunctionCall(Hue Shift Percentage=@n27, Texture=@n26) {materialFunction=&HueShift}.Result
    Alpha=NamedRerouteUsage()
   ) {constB=0}) {name=FinalColor, nodeColor=<rgba 0.875 1 0 1>}
  n28 = Multiply(A=NamedRerouteUsage()) {constB=0.8}
  NamedRerouteDeclaration(None=Multiply(A=NamedRerouteUsage(), B=OneMinus(None=@n28))) {name=AlphaNoBlades, nodeColor=<rgba 0 1 0.8594 1>}
  n29 = PreSkinnedPosition()
  n30 = Constant3Vector() {constant=<rgba 0 0 9 0>}
  n31 = VertexInterpolator(VS=Step(X=Add(A=@n29, B=@n30)))
  NamedRerouteDeclaration(None=OneMinus(None=ComponentMask(None=@n31.PS) {b=true})) {name=LowerHeadFoneMask, nodeColor=<rgba 0 0.1094 1 1>}
  NamedRerouteDeclaration(None=ComponentMask(None=@n8) {r=true}) {name=TattooMask, nodeColor=<rgba 0 1 0.4844 1>}
  n32 = PreSkinnedPosition()
  n33 = Constant3Vector()
  n34 = VertexInterpolator(VS=Divide(A=Add(A=@n32, B=@n33), B=Constant3Vector() {constant=<rgba 11 1 1 0>}))
  n35 = ComponentMask(None=Abs(None=@n34.PS)) {r=true}
  n36 = ComponentMask(None=@n34.PS) {b=true}
  NamedRerouteDeclaration(None=Multiply(A=Step(X=@n35) {constY=0.85}, B=Saturate(None=@n36))) {name=PvtWeaponMask, nodeColor=<rgba 0 1 0.0625 1>}
  n37 = Multiply(A=VectorParameter() {parameterName=GlowFresnel, defaultValue=<rgba 1.039 0 2 1>}.RGB, B=NamedRerouteUsage())
  n38 = Power(Base=NamedRerouteUsage()) {constExponent=3}
  NamedRerouteDeclaration(None=Multiply(A=Multiply(A=@n37, B=@n38), B=Constant() {r=12}) {constB=12}) {name=Fresnel Line, nodeColor=<rgba 0.2656 0 1 1>}
  n39 = VectorParameter() {group=ChicleVeil, parameterName=GlowFresnel, defaultValue=<rgba 1.039 0 2 1>}
  n40 = NamedRerouteUsage()
  NamedRerouteDeclaration(None=Multiply(A=LinearInterpolate(A=@n39.RGB, Alpha=@n40), B=NamedRerouteUsage())) {name=Frenel Base, nodeColor=<rgba 0.7344 0 1 1>}
  NamedRerouteDeclaration(None=Multiply(A=Constant3Vector() {constant=<rgba 0.3698 0.0643 0.309 1>}, B=NamedRerouteUsage())) {name=CrystalColor, nodeColor=<rgba 1 0 0.8438 1>}
  n41 = Multiply(
   A=MaterialFunctionCall(Hue Shift Percentage=Constant() {r=0.35}, Texture=NamedRerouteUsage()) {materialFunction=&HueShift}.Result
   B=NamedRerouteUsage()
  )
  n42 = VectorParameter() {group=ChicleVeil, parameterName=Weapon ChicleVeil, defaultValue=<rgba 0.5006 0.0177 0.5885 1>}
  NamedRerouteDeclaration(None=Multiply(A=Multiply(A=@n41, B=@n42.RGB)) {constB=0.6}) {name=WeaponEmissiveColor, nodeColor=<rgba 0.5703 0 1 1>}
  n43 = Multiply(A=NamedRerouteUsage(), B=NamedRerouteUsage())
  n44 = VectorParameter() {group=ChicleVeil, parameterName=GlowFresnel, defaultValue=<rgba 1.039 0 2 1>}
  n45 = Multiply(A=NamedRerouteUsage()) {constB=0.2}
  NamedRerouteDeclaration(None=Multiply(A=Multiply(B=@n43) {constA=5}, B=Multiply(A=@n44.RGB, B=@n45))) {name=TattooGlow, nodeColor=<rgba 1 0.5391 0 1>}
  NamedRerouteDeclaration(None=@n16.MaterialAttributes) {name=MA, nodeColor=<rgba 0 0.5312 1 1>}
  NamedRerouteDeclaration(None=FunctionInput() {inputName=GlowAmount, inputType=FunctionInput_Scalar, sortPriority=3}) {name=GlowAmount, nodeColor=<rgba 1 0 0.4219 1>}
  NamedRerouteDeclaration(None=FunctionInput() {inputName=OnKill, inputType=FunctionInput_Scalar, sortPriority=4}) {name=OnKill, nodeColor=<rgba 0.7109 1 0 1>}
  Reroute(None=@n24)
 ]}
