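# Skeletal mesh material for the ChicleVeil backpack cosmetic, with glow and crystal effects.
# The shared 2023 character PreFX function feeds a backpack function driven by emissive and on-kill emissive parameters and an FX mask texture.
# A crystal function follows, taking that texture's green channel as alpha, before the shared PostFX function finishes the result.
# It uses material attributes, with shading model chosen by expression.
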Material /BRCosmetics/Accessories/FORT_Backpacks/Backpack_F_MED_ChicleVeil/Materials/M_F_MED_ChicleVeil_Backpack
graph {
 properties {
  bUseMaterialAttributes=true
  bUsedWithMeshDeformer=true
  bUsedWithSkeletalMesh=true
  pixelDepthOffsetMode=PDOM_Legacy
  shadingModel=MSM_FromMaterialExpression
  usedShadingModels=Default Lit | Subsurface | Clear Coat | Subsurface Profile | Cloth
 }
 nodes [
  n1 = MaterialFunctionCall() {materialFunction=&MF_CharacterParent_2023_PreFX}
  n2 = ScalarParameter() {group=ChicleVeil, parameterName=VibrantEmissiveAmount}
  n3 = ScalarParameter() {group=ChicleVeil, parameterName=OnKillVibrantEmissiveAmount}
  n4 = TextureSample() {texture=/BRCosmetics/Accessories/FORT_Backpacks/Backpack_F_MED_ChicleVeil/Textures/T_ChicleVeil_Backpack_FX2}
  MaterialFunctionCall(
   Material=MaterialFunctionCall!(
    Material=MaterialFunctionCall!(In=@n1.Result, GlowAmount=@n2, OnKill=@n3, TAMask2=@n4.RGB) {materialFunction=&/MF_Chicleveil_Backpack}.Result
    Alpha=@n4.G
   ) {materialFunction=&/MF_ChicleVeil_Crystal}.Result
   M=@n1.M
   Pre Skinned Local Position=@n1.M
   Distance Blend=@n1.M
   Fresnel=@n1.M
  ) {materialFunction=&MF_CharacterParent_2023_PostFX}
 ]
}
