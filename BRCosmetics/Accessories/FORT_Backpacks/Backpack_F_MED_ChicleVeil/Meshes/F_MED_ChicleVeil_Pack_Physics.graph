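# Physics asset for the ChicleVeil backpack cosmetic mesh, giving its dynamic bones capsule bodies for secondary motion.
# Eight capsules cover the pack center, left and right pack pieces, two weapons, and a headphone band with two ear pieces.
# Constraints form a tree from the pack center: weapons hang off the side pieces and ear pieces off the band.
# Tight swing and twist limits, with the second swing axis at zero, allow only slight jiggle.
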PhysicsAsset /BRCosmetics/Accessories/FORT_Backpacks/Backpack_F_MED_ChicleVeil/Meshes/F_MED_ChicleVeil_Pack_Physics
bodies {
 dyn_headphones [{shapeType=Capsule, length=5.1453, radius=2.1505, shapeName=dyn_headphones_capsule, rotation=<rot -90 90 -90.0005>}]
 dyn_headphones_l [{shapeType=Capsule, length=4.3347, radius=3.0189, shapeName=dyn_headphones_l_capsule}]
 dyn_headphones_r [{shapeType=Capsule, length=4.3426, radius=3.0283, shapeName=dyn_headphones_r_capsule}]
 dyn_pack_center [{shapeType=Capsule, length=5.8896, radius=5.0031, shapeName=dyn_pack_center_capsule}]
 dyn_pack_l [{shapeType=Capsule, length=2.6327, radius=2.0999, shapeName=dyn_pack_l_capsule}]
 dyn_pack_r [{shapeType=Capsule, length=2.6326, radius=2.1129, shapeName=dyn_pack_r_capsule}]
 dyn_weapon_l [{shapeType=Capsule, length=12.4507, radius=4.9664, shapeName=dyn_weapon_l_capsule}]
 dyn_weapon_r [{shapeType=Capsule, length=11.9728, radius=4.2369, shapeName=dyn_weapon_r_capsule}]
}
constraints [
 {bone1Name=dyn_headphones_l, bone2Name=dyn_headphones, swing1LimitDegrees=1, swing2LimitDegrees=0, twistLimitDegrees=3}
 {bone1Name=dyn_headphones_r, bone2Name=dyn_headphones, swing1LimitDegrees=1, swing2LimitDegrees=0, twistLimitDegrees=3}
 {bone1Name=dyn_weapon_l, bone2Name=dyn_pack_l, swing1LimitDegrees=3, swing2LimitDegrees=0, twistLimitDegrees=3}
 {bone1Name=dyn_weapon_r, bone2Name=dyn_pack_r, swing1LimitDegrees=3, swing2LimitDegrees=0, twistLimitDegrees=3}
 {bone1Name=dyn_pack_r, bone2Name=dyn_pack_center, swing1LimitDegrees=3, swing2LimitDegrees=0, twistLimitDegrees=3}
 {bone1Name=dyn_pack_l, bone2Name=dyn_pack_center, swing1LimitDegrees=3, swing2LimitDegrees=0, twistLimitDegrees=3}
 {bone1Name=dyn_headphones, bone2Name=dyn_pack_center, swing1LimitDegrees=5, swing2LimitDegrees=0, twistLimitDegrees=5}
]
